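# Translucent, unlit, two-sided material for Niagara mesh particles that draws an eroding explosion shockwave ring.
# Two noise textures pan in radial coordinates with UV1-driven warping and per-particle time offsets, merging to distort a soft band mask.
# A smoothstep erodes that noise to drive particle-tinted emissive and depth-faded opacity, with a low-quality path skipping the fade.
# Dynamic parameters control erosion, band shift, extra panning and a vertex pinch toward the particle center.
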Material /Game/Effects/Niagara/MasterMaterials/Explosion/M_Shockwave_Ring
graph {
 properties {
  bUsedWithNiagaraMeshParticles=true
  blendMode=BLEND_Translucent
  pixelDepthOffsetMode=PDOM_Legacy
  shadingModel=MSM_Unlit
  twoSided=true
  usedShadingModels=Unlit
 }
 nodes [
  ActorPositionWS()
  MaterialFunctionCall() {materialFunction=&DebugTimeSine}
  n1 = ParticleColor()
  n2 = OneMinus(None=ComponentMask(None=TextureCoordinate() {coordinateIndex=1}) {g=true})
  n3 = VectorParameter() {parameterName=TileAndPan2, defaultValue=<rgba 2 -0.05 -0.05 0.05>}
  n4 = ComponentMask(None=TextureCoordinate() {coordinateIndex=1}) {g=true}
  n5 = LinearInterpolate(
   B=LinearInterpolate(Alpha=Power(Base=@n4) {constExponent=0.6}) {constA=1, constB=0.2}
   Alpha=ScalarParameter() {defaultValue=1, parameterName=UVWarp}
  ) {constA=1}
  n6 = AppendVector(A=@n3.R, B=Multiply(A=@n3.G, B=@n5))
  n7 = MaterialFunctionCall() {materialFunction=&VectorToRadialValue}
  n8 = ParticleRandom()
  n9 = DynamicParameter() {defaultValue=<rgba 0 0.6 0 1>, paramNames [Erode,Shift,WPOPinch,PanAdd]}
  n10 = Add(A=Add(A=Time(), B=Multiply(A=@n8) {constB=10}), B=@n9.PanAdd)
  n11 = TextureCoordinate()
  n12 = DDX(Value=@n11)
  n13 = DDY(Value=@n11)
  n14 = TextureSampleParameter2D!(
   UVs=Panner(Coordinate=Multiply(A=@n6, B=@n7.Radial Coordinates), Time=@n10, Speed=AppendVector(A=@n3.B, B=@n3.B))
   "DDX(UVs)"=@n12
   "DDY(UVs)"=@n13
   Apply View MipBias
  ) {mipValueMode=TMVM_Derivative, parameterName=Texture2, texture=/Game/Effects/Fort_Effects/Textures/T_AtmosphericCloudNoise_Vertical}
  n15 = VectorParameter() {parameterName=TileAndPan1, defaultValue=<rgba 3 -0.17 0.05 0.2>}
  n16 = AppendVector(A=@n15.R, B=Multiply(A=@n15.G, B=@n5))
  n17 = TextureSampleParameter2D(
   UVs=Panner(Coordinate=Multiply(A=@n7.Radial Coordinates, B=@n16), Time=@n10, Speed=AppendVector(A=@n15.B, B=@n15.B))
   "DDX(UVs)"=@n12
   "DDY(UVs)"=@n13
  ) {
   mipValueMode=TMVM_Derivative
   parameterName=Texture1
   samplerType=SAMPLERTYPE_Grayscale
   texture=/Game/Effects/Fort_Effects/Textures/Plasma/T_Turbulence_Stretched_Balanced
  }
  n18 = Max(A=Multiply(A=@n14.G) {constB=0.85}, B=@n17.G) {constB=0}
  n19 = Subtract(A=@n18) {constB=0.5}
  n20 = ScalarParameter() {defaultValue=-0.4, parameterName=TextureNoise}
  n21 = SphereMask(A=Add(A=Multiply(A=@n19, B=@n20), B=Power(Base=@n4, Exp=@n9.Shift) {constExponent=0.4}), B=Constant() {r=0.5}) {attenuationRadius=0.42, hardnessPercent=0}
  n22 = ComponentMask(None=TextureCoordinate() {coordinateIndex=1}) {r=true}
  n23 = Constant() {r=0.5}
  n24 = StaticSwitchParameter(True=Multiply(A=@n21, B=SphereMask(A=@n22, B=@n23, Radius=@n23) {attenuationRadius=0.5, hardnessPercent=50}), False=@n21) {parameterName=EdgeFade}
  n25 = Power(
   Base=@n18
   Exp=LinearInterpolate(
    A=ScalarParameter() {defaultValue=1.5, parameterName=PowerEdge}
    B=ScalarParameter() {defaultValue=0.75, parameterName=PowerCenter}
    Alpha=@n24
   ) {constA=2, constB=0.5}
  )
  n26 = SmoothStep(Min=@n25, Max=Add(A=@n25) {constB=-0.5}, Value=NamedRerouteUsage())
  n27 = Multiply(
   A=@n1.RGB
   B=Add(A=VectorParameter() {parameterName=BackgroundColor, defaultValue=<rgba 0.1099 0.1154 0.2 1>}.RGB, B=Multiply(A=@n2, B=@n26))
  )
  n28 = MaterialFunctionCall(LDR Input=Clamp(None=@n27), HDR Tint=@n27) {materialFunction=&DeriveHDRfromLDR}
  MaterialFunctionCall(Emissive=QualitySwitch(Default=ShadingPathSwitch(Default=@n27, Deferred=@n28.RemappedRange), Low=@n27)) {materialFunction=&MF_FX_EmissiveAdjustments}
  n29 = ParticlePositionWS()
  n30 = WorldPosition()
  n31 = TextureCoordinate() {coordinateIndex=1}
  Multiply(A=Multiply(A=Subtract(A=@n29, B=@n30.XYZ), B=ComponentMask(None=@n31) {g=true}), B=@n9.WPOPinch)
  n32 = Multiply(A=@n26, B=Saturate(None=Multiply(A=@n24)))
  n33 = LinearInterpolate(
   A=ScalarParameter() {defaultValue=200, parameterName=DepthFadeMax}
   B=ScalarParameter() {defaultValue=10, parameterName=DepthFadeMin}
   Alpha=@n25
  )
  Multiply(A=@n1.A, B=QualitySwitch(Default=DepthFade(Opacity=@n32, FadeDistance=@n33), Low=@n32))
  NamedRerouteDeclaration(None=@n9.Erode) {name=Erode, nodeColor=<rgba 0 0.7187 1 1>}
 ]
}
